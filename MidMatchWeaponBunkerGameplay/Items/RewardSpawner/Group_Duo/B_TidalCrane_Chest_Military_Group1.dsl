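; Military chest reward spawner Blueprint that scales its spawn chance level to the current playlist's team size.
; On BeginPlay, the server waits asynchronously for playlist data, clamps max team size between one and four, and stores it.
; It sets the spawn chance level to team size minus one, then calls the parent BeginPlay, so parent logic runs only on the server once the playlist is ready.
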(event EventBeginPlay
  ; Begin Play
  (if (Networking|IsServer)
    (bind (_gamestate _playlist) (Class|FortAsyncActionCurrentPlaylistReady|PlaylistDataReadyAsync)
      (:OnReady
        (bind _output_get (Variables|Default|SetMax%20Team%20Size (Math|Integer|Clamp(Integer) (Variables|/Script/FortniteGame.FortPlaylist|GetMaxTeamSize _playlist) 1 4)))
        (Variables|SpawnerVars|SetSpawnChanceLevel (Math|Conversions|ToFloat(Integer) (- _output_get 1)))
        (|CallParentFunction)))))

(event Collision|EventActorBeginOverlap
  (params (OtherActor :type Actor%20Object%20Reference)))

(event EventTick
  (params (DeltaSeconds :type Float%20%28single-precision%29)))

(fn ConstructionScript
  (|CallParentFunction))
